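; BP_Decal_Reality_Glitch is a Blueprint for what its name indicates is a reality-glitch decal effect with looping audio.
; On EndPlay it takes the RealityAudioLoop audio component and briefly fades it to silence with a linear curve.
; BeginPlay, Tick and the construction script are empty, so the audio fade-out is the only scripted logic shown.
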(event EventBeginPlay)

(event EventTick
  (params (DeltaSeconds :type Float%20%28single-precision%29)))

(event EventEndPlay
  (params (EndPlayReason :type EEndPlayReason%20Enum :default "Destroyed"))
  (bind _realityaudioloop (Variables|Default|GetRealityAudioLoop)
    (:then
      (Audio|Components|Audio|FadeOut _realityaudioloop 0.35 0.0 "Linear" :!class /Script/Engine.AudioComponent))
    (:else)))

(fn ConstructionScript)
