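# Unlit translucent deferred decal material for blast marks, built on a ground char texture.
# A time-of-day function selects between a morning/night color parameter and a day/evening one.
# Two half-scale, offset samples of tiling smoke are summed, power-adjusted and multiplied by the center-scaled char mask and an Opacity parameter.
# A separate char-mask power and Metallic, Specular and Roughness parameters also exist, with no material output links shown.
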Material /Game/Effects/Fort_Effects/Decals/Materials/MD_Decal_Blast
graph {
 properties {
  bAutomaticallySetUsageInEditor=false
  blendMode=BLEND_Translucent
  materialDomain=MD_DeferredDecal
  pixelDepthOffsetMode=PDOM_Legacy
  shadingModel=MSM_Unlit
  usedShadingModels=Unlit
 }
 nodes [
  Add()
  n1 = VectorParameter() {parameterName=DecalColor_MorningAndNight, defaultValue=<rgba 0.8963 0.2453 0.0381 1>}
  n2 = VectorParameter() {parameterName=DecalColor, defaultValue=<rgba 0.8963 0.2406 0.0516 1>}
  MaterialFunctionCall(Color_Morning=@n1.RGB, Color_Day=@n2.RGB, Color_Evening=@n2.RGB, Color_Night=@n1.RGB) {materialFunction=&Color_Time_of_Day}
  n3 = TextureCoordinate()
  n4 = Multiply(A=TextureCoordinate()) {constB=0.5}
  n5 = Constant2Vector() {r=0.5}
  n6 = Add(
   A=TextureSample(UVs=Multiply(A=@n3) {constB=0.5}) {texture=&T_Inky_Smoke_Tile}.R
   B=TextureSample(UVs=Add(A=@n4, B=@n5)) {texture=&T_Inky_Smoke_Tile}.R
  )
  n7 = TextureSample(UVs=MaterialFunctionCall(UVs=TextureCoordinate(), Texture Scale=ScalarParameter() {defaultValue=1.4, parameterName=ScaleOpacity}) {materialFunction=&ScaleUVsByCenter}.UVs) {samplerType=SAMPLERTYPE_LinearColor, texture=/Game/Effects/Fort_Effects/Gameplay/SmokeDustExplosions/Textures/T_GroundChar_01}
  Multiply(A=Multiply(A=Power(Base=@n6) {constExponent=0.25}, B=@n7.G), B=ScalarParameter() {defaultValue=1, parameterName=Opacity})
  Power(Base=@n7.G, Exp=ScalarParameter() {defaultValue=1, parameterName=PowerOpacity}) {constExponent=0.25}
  ScalarParameter() {parameterName=Metallic}
  ScalarParameter() {defaultValue=1, parameterName=Specular}
  ScalarParameter() {defaultValue=1, parameterName=Roughness}
 ]
}
