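# Unlit translucent deferred decal material, named for the mark a snowball impact leaves on world surfaces.
# Two offset samples of a tiling smoke texture are summed, softened by a power, then masked by a ground char texture's green channel.
# An Erode_Alpha function erodes this mask using inverted decal lifetime opacity scaled by ErodeStrength.
# Color comes from a time-of-day function, pairing one tint for morning and night and another for day and evening, multiplied with the mask.
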Material /Game/Effects/Fort_Effects/Decals/Materials/MD_Decal_Snowball_WorldImpact
graph {
 properties {
  bAutomaticallySetUsageInEditor=false
  blendMode=BLEND_Translucent
  materialDomain=MD_DeferredDecal
  pixelDepthOffsetMode=PDOM_Legacy
  shadingModel=MSM_Unlit
  usedShadingModels=Unlit
 }
 nodes [
  n1 = DecalLifetimeOpacity()
  n2 = TextureSample(UVs=Multiply(A=TextureCoordinate()) {constB=0.5}) {texture=&T_Inky_Smoke_Tile}
  n3 = TextureCoordinate()
  n4 = TextureSample(UVs=Add(A=Multiply(A=@n3) {constB=0.5}, B=Constant2Vector() {r=0.5})) {texture=&T_Inky_Smoke_Tile}
  n5 = Multiply(A=TextureCoordinate())
  n6 = Constant2Vector()
  n7 = Multiply(
   A=Power(Base=Add(A=@n2.R, B=@n4.R)) {constExponent=0.25}
   B=TextureSample(UVs=Add(A=@n5, B=@n6)) {samplerType=SAMPLERTYPE_LinearColor, texture=/Game/Effects/Fort_Effects/Gameplay/SmokeDustExplosions/Textures/T_GroundChar_01}.G
  )
  MaterialFunctionCall!(Dynamic Param=Multiply(A=OneMinus(None=@n1.Opacity), B=ScalarParameter() {defaultValue=-5, parameterName=ErodeStrength}), Alpha=@n7) {materialFunction=/Game/Effects/Fort_Effects/Materials/_Functions/Erode_Alpha}
  n8 = VectorParameter() {parameterName=DecalColor_MorningAndNight, defaultValue=<rgba 1 1 1 1>}
  n9 = VectorParameter() {parameterName=DecalColor, defaultValue=<rgba 1 1 1 1>}
  n10 = MaterialFunctionCall(Color_Morning=@n8.RGB, Color_Day=@n9.RGB, Color_Evening=@n9.RGB, Color_Night=@n8.RGB) {materialFunction=&Color_Time_of_Day}
  Multiply(A=@n10.Result, B=Multiply(A=@n10.Result, B=@n7))
  ScalarParameter() {defaultValue=1, parameterName=Metallic}
  ScalarParameter() {defaultValue=1, parameterName=Specular}
  ScalarParameter() {parameterName=Roughness}
 ]
}
