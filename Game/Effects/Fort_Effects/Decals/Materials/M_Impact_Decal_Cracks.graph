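# Translucent deferred decal material that draws cracked concrete for impact effects.
# A rotatable, tiled crack texture plus road-crack detail forms a pattern that, multiplied by a mask texture, drives opacity, faded by decal lifetime, facing and distance along the projection axis.
# Concrete color is modulated by the detail cracks, and two normal maps are combined and scaled.
# A time-and-duration staggered-gradient term offsets the crack pattern in a branch separate from opacity.
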Material /Game/Effects/Fort_Effects/Decals/Materials/M_Impact_Decal_Cracks
graph {
 properties {blendMode=BLEND_Translucent, materialDomain=MD_DeferredDecal, pixelDepthOffsetMode=PDOM_Legacy, usedShadingModels=Default Lit}
 nodes [
  n1 = Multiply(A=TextureCoordinate()) {constB=2}
  n2 = TextureSample(UVs=@n1) {samplerType=SAMPLERTYPE_Masks, texture=/Game/Athena/Apollo/Environments/Landscape/Textures/Roads/Road_Cracks_M}
  n3 = OneMinus(None=@n2.G)
  n4 = Multiply(A=TextureCoordinate(), B=ScalarParameter() {defaultValue=3.5, parameterName=Tile})
  n5 = TextureSampleParameter2D(UVs=MaterialFunctionCall(UVs=@n4, "Rotation Angle (0-1)"=ScalarParameter() {parameterName=RotateBaseTex}) {materialFunction=&CustomRotator}.Rotated Values) {parameterName=BaseTexture, samplerType=SAMPLERTYPE_Grayscale, texture=/Game/Effects/Fort_Effects/Textures/Tile/T_Concrete_Crack_A}
  n6 = Add(A=Multiply(A=Multiply(A=@n3, B=@n3), B=ScalarParameter() {defaultValue=0.5, parameterName=DetailCrackMult}) {constB=0.25}, B=@n5.G)
  n7 = Subtract(A=Time(), B=ScalarParameter() {parameterName=Time})
  n8 = ScalarParameter() {defaultValue=0.5, parameterName=Duration}
  n9 = OneMinus(None=Saturate(None=Divide(A=@n7, B=@n8)))
  Add(A=@n6, B=Subtract(A=TextureSample(UVs=@n9) {texture=/Game/Effects/Fort_Effects/Textures/Grad/T_Staggered_Gradient}.G))
  n10 = TextureSampleParameter2D(UVs=@n4) {parameterName=BaseDiffuse, texture=/Game/Environments/Apollo/Sets/Coliseum_Ruins/Textures/T_CR_Concrete_Cracked_D}
  Multiply(
   A=Multiply(A=LinearInterpolate(Alpha=@n2.G) {constA=0.3}, B=Multiply(A=@n10.RGB, B=@n10.RGB))
   B=ScalarParameter() {defaultValue=0.4, parameterName=BaseColorBrightness}
  ) {constB=0.7}
  n11 = TextureSampleParameter2D(UVs=@n4) {
   parameterName=BaseNormal
   samplerType=SAMPLERTYPE_Normal
   texture=/Game/Environments/Apollo/Sets/Coliseum_Ruins/Textures/T_CR_Concrete_Cracked_N
  }
  n12 = TextureSampleParameter2D(UVs=@n1) {parameterName=NormalTexture, samplerType=SAMPLERTYPE_Normal, texture=/Game/Athena/Apollo/Environments/Landscape/Textures/Roads/Road_Cracks_N}
  n13 = Constant3Vector() {constant=<rgba 0.5 0.5 0 0>}
  Multiply(A=Add(A=@n11.RGB, B=Multiply(A=@n12.RGB, B=@n13)), B=VectorParameter() {parameterName=NormalMult, defaultValue=<rgba 3 3 1 1>}.RGB)
  n14 = TextureSampleParameter2D() {
   parameterName=MaskTexture
   samplerType=SAMPLERTYPE_LinearColor
   texture=/Game/Effects/Fort_Effects/Gameplay/SmokeDustExplosions/Textures/T_GroundChar_01
  }
  n15 = Saturate(None=@n6)
  n16 = Saturate(None=Multiply(A=Multiply(A=@n14.G, B=@n15), B=ScalarParameter() {defaultValue=10, parameterName=OpacityMult}))
  n17 = DecalLifetimeOpacity()
  n18 = WorldPosition()
  n19 = ObjectPositionWS()
  n20 = Abs(None=DotProduct(A=Subtract(A=@n18.XYZ, B=@n19), B=ObjectOrientation()))
  n21 = Multiply(A=ScalarParameter() {parameterName=Fade_TopBottom_Offset}, B=@n16)
  n22 = Saturate(None=Divide(A=Add(A=@n20, B=@n21), B=ScalarParameter() {defaultValue=8, parameterName=Fade_TopBottom}))
  Multiply(
   A=Multiply(A=Multiply(A=@n16, B=@n17.Opacity), B=OneMinus(None=@n22))
   B=MaterialFunctionCall() {materialFunction=/Game/Weapons/Prototype/SprayGun/Materials/MF_DecalFacingMask}.Result
  )
  ScalarParameter() {defaultValue=0.85, parameterName=Roughness}
  ScalarParameter() {defaultValue=0.5, parameterName=Spec}
 ]
}
